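# Masked, two-sided bat material for Niagara mesh particles in the Bats aura cosmetic, animating the mesh procedurally in the shader.
# UV1 and UV2 masks separate wings (left/right), body and legs for sine-driven flapping, bending, body roll, leg/tail swing and wobble via axis rotations and offsets.
# Niagara dynamic parameters control speeds, amounts and the MoveBodyOrWing blend between two motion sets.
# Color comes from a hue-shifted wing texture multiplied by particle color.
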Material /SparksCosmetics/Items/Auras/Assets/NS_Aura_Bats/M_Bat_Black
graph {
 properties {
  bTangentSpaceNormal=false
  bUsedWithNiagaraMeshParticles=true
  blendMode=BLEND_Masked
  opacityMaskClipValue=0.333
  pixelDepthOffsetMode=PDOM_Legacy
  twoSided=true
  usedShadingModels=Default Lit
 }
 nodes [
  Constant()
  Constant() {r=0.1}
  Constant() {desc=Tail Rot Amount, r=5}
  DynamicParameter() {defaultValue=<rgba 0.25 1 0.25 0.1>, paramNames [MoveBodyOrWing, BatAnimFreq, WingBend, WingMiddleBend]}
  DynamicParameter() {parameterIndex=1, defaultValue=<rgba 1 10 2 0.25>, paramNames [TinyWobbleSpeed, UpDownMove, DiagnolBackForthMove, Param4]}
  n1 = Transform(None=Constant3Vector() {constant=<rgba 1 0 0.5 0>}) {transformSourceType=TRANSFORMSOURCE_Local}
  n2 = DynamicParameter() {defaultValue=<rgba 0.25 1 0.25 0.1>, paramNames [MoveBodyOrWing, BatAnimFreq, WingBend, WingMiddleBend]}
  n3 = Multiply(A=Multiply(A=Time(), B=NamedRerouteUsage()), B=Multiply(A=Constant() {desc=Body Anim Speed, r=1}, B=@n2.BatAnimFreq))
  n4 = DynamicParameter() {parameterIndex=1, defaultValue=<rgba 1 10 2 0.25>, paramNames [TinyWobbleSpeed, UpDownMove, DiagnolBackForthMove, Param4]}
  n5 = Multiply(A=Multiply(A=Normalize(VectorInput=@n1), B=Sine(None=@n3)), B=@n4.DiagnolBackForthMove)
  n6 = NamedRerouteUsage()
  n7 = Clamp(None=Add(A=OneMinus(None=@n6), B=NamedRerouteUsage()))
  n8 = Normalize(VectorInput=Transform(None=Constant3Vector() {constant=<rgba -0.5 0 1 0>}) {transformSourceType=TRANSFORMSOURCE_Local})
  n9 = Multiply(A=Time(), B=NamedRerouteUsage())
  n10 = Constant() {r=0.2}
  n11 = Constant() {desc=Leg Anim Speed, r=2}
  n12 = DynamicParameter() {defaultValue=<rgba 0.25 1 0.25 0.1>, paramNames [MoveBodyOrWing, BatAnimFreq, WingBend, WingMiddleBend]}
  n13 = Sine(None=Multiply(A=Add(A=@n9, B=@n10), B=Multiply(A=@n11, B=@n12.BatAnimFreq)))
  n14 = Multiply(A=Multiply(A=Multiply(A=@n8, B=@n13), B=Constant() {desc=Leg Anim Amount, r=2}), B=Multiply(A=NamedRerouteUsage(), B=NamedRerouteUsage()))
  n15 = Transform(None=Constant3Vector() {constant=<rgba 0 0 1 0>}) {transformSourceType=TRANSFORMSOURCE_Local}
  n16 = Multiply(A=Multiply(A=Normalize(VectorInput=@n15), B=Constant() {desc=Tail Move Amount, r=2}), B=@n13)
  n17 = Multiply(A=NamedRerouteUsage(), B=NamedRerouteUsage())
  n18 = Normalize(VectorInput=Transform(None=Constant3Vector() {constant=<rgba 1 0 0 0>}) {transformSourceType=TRANSFORMSOURCE_Local})
  n19 = DynamicParameter() {parameterIndex=2, defaultValue=<rgba 1 1 2 0.25>, paramNames [BodyRollSpeed, BodyRollAmount, Param3, Param4]}
  n20 = DynamicParameter() {defaultValue=<rgba 0.25 1 0.25 0.1>, paramNames [MoveBodyOrWing, BatAnimFreq, WingBend, WingMiddleBend]}
  n21 = Multiply(A=Multiply(A=Time(), B=NamedRerouteUsage()), B=Multiply(A=@n19.BodyRollSpeed, B=@n20.BatAnimFreq))
  n22 = DynamicParameter() {parameterIndex=2, defaultValue=<rgba 1 1 2 0.25>, paramNames [BodyRollSpeed, BodyRollAmount, Param3, Param4]}
  n23 = Multiply(A=Multiply(A=Sine(None=@n21), B=@n22.BodyRollAmount), B=@n17)
  n24 = MaterialFunctionCall() {materialFunction=&ComponentPivotLocation}
  n25 = WorldPosition()
  n26 = Transform(None=LinearInterpolate(
    A=VectorParameter() {parameterName=WingLeft, defaultValue=<rgba 1 -0.25 -0.5 1>}.RGB
    B=VectorParameter() {parameterName=WingRight, defaultValue=<rgba -1 -0.25 0.5 1>}.RGB
    Alpha=NamedRerouteUsage()
   )) {transformSourceType=TRANSFORMSOURCE_Local}
  n27 = NamedRerouteUsage()
  n28 = Constant() {r=0.2}
  n29 = Multiply(A=Clamp(None=Add(A=@n27, B=@n28)), B=NamedRerouteUsage())
  n30 = DynamicParameter() {defaultValue=<rgba 0.25 1 0.25 0.1>, paramNames [MoveBodyOrWing, BatAnimFreq, WingBend, WingMiddleBend]}
  n31 = NamedRerouteUsage()
  n32 = Add(A=Multiply(A=Time(), B=NamedRerouteUsage()), B=Multiply(A=@n30.WingMiddleBend, B=OneMinus(None=@n31)))
  n33 = DynamicParameter() {defaultValue=<rgba 0.25 1 0.25 0.1>, paramNames [MoveBodyOrWing, BatAnimFreq, WingBend, WingMiddleBend]}
  n34 = DynamicParameter() {parameterIndex=3, defaultValue=<rgba 1 1 0.25 1>, paramNames [Wing_FlapSpeed, Param2, Param3, MainSpeed]}
  n35 = Multiply(A=@n33.BatAnimFreq, B=@n34.Wing_FlapSpeed)
  n36 = DynamicParameter() {defaultValue=<rgba 0.25 1 0.25 0.1>, paramNames [MoveBodyOrWing, BatAnimFreq, WingBend, WingMiddleBend]}
  n37 = Multiply(A=Sine(None=Multiply(A=@n32, B=@n35)), B=@n36.WingBend)
  n38 = Add(
   A=Add(A=Multiply(A=@n16, B=@n17), B=RotateAboutAxis(NormalizedRotationAxis=@n18, RotationAngle=@n23, PivotPoint=@n24.Result, Position=@n25.XYZ))
   B=RotateAboutAxis(
    NormalizedRotationAxis=Normalize(VectorInput=@n26)
    RotationAngle=Multiply(A=@n29, B=@n37)
    PivotPoint=MaterialFunctionCall() {materialFunction=&ComponentPivotLocation}.Result
    Position=WorldPosition().XYZ
   )
  )
  n39 = Constant3Vector() {constant=<rgba 0.25 1 0.25 0>}
  n40 = Time()
  n41 = NamedRerouteUsage()
  n42 = DynamicParameter() {defaultValue=<rgba 0.25 1 0.25 0.1>, paramNames [MoveBodyOrWing, BatAnimFreq, WingBend, WingMiddleBend]}
  n43 = Sine(None=Multiply(A=Multiply(A=@n40, B=@n41), B=@n42.BatAnimFreq))
  n44 = Constant() {desc=Wing Anim Amount, r=0.25}
  n45 = NamedRerouteUsage()
  n46 = Multiply(A=Constant() {r=0.2}, B=NamedRerouteUsage())
  n47 = RotateAboutAxis(
   NormalizedRotationAxis=Normalize(VectorInput=Transform(None=@n39) {transformSourceType=TRANSFORMSOURCE_Local})
   RotationAngle=Multiply(A=Multiply(A=@n43, B=@n44), B=Add(A=@n45, B=@n46))
   PivotPoint=MaterialFunctionCall() {materialFunction=&ComponentPivotLocation}.Result
   Position=WorldPosition().XYZ
  )
  n48 = NamedRerouteUsage()
  n49 = Constant3Vector() {constant=<rgba 0 0 1 0>}
  n50 = Add(A=Multiply(A=Time(), B=NamedRerouteUsage()), B=Multiply(A=Constant() {r=-0.1}, B=NamedRerouteUsage()))
  n51 = DynamicParameter() {parameterIndex=1, defaultValue=<rgba 1 10 2 0.25>, paramNames [TinyWobbleSpeed, UpDownMove, DiagnolBackForthMove, Param4]}
  n52 = DynamicParameter() {defaultValue=<rgba 0.25 1 0.25 0.1>, paramNames [MoveBodyOrWing, BatAnimFreq, WingBend, WingMiddleBend]}
  n53 = Multiply(A=@n51.TinyWobbleSpeed, B=@n52.BatAnimFreq)
  n54 = Multiply(A=Normalize(VectorInput=Transform(None=@n49) {transformSourceType=TRANSFORMSOURCE_Local}), B=Sine(None=Multiply(A=@n50, B=@n53)))
  n55 = DynamicParameter() {parameterIndex=1, defaultValue=<rgba 1 10 2 0.25>, paramNames [TinyWobbleSpeed, UpDownMove, DiagnolBackForthMove, Param4]}
  n56 = DynamicParameter() {defaultValue=<rgba 0.25 1 0.25 0.1>, paramNames [MoveBodyOrWing, BatAnimFreq, WingBend, WingMiddleBend]}
  LinearInterpolate(
   A=Add(A=Multiply(A=@n5, B=@n7), B=Add(A=@n14, B=@n38))
   B=Add(A=Multiply(A=@n47, B=@n48), B=Multiply(A=@n54, B=@n55.UpDownMove))
   Alpha=@n56.MoveBodyOrWing
  )
  n57 = TextureSampleParameter2D() {parameterName=Diffuse Texture, texture=/Game/Effects/Fort_Effects/Effects/Environment/Jungle/Textures/T_Bat_Wing_FX_01}
  Multiply(
   A=ParticleColor().RGB
   B=MaterialFunctionCall(
    Hue Shift Percentage=ScalarParameter() {parameterName=Hue Shift}
    Texture=MaterialFunctionCall(DiffuseColor=@n57.RGB) {materialFunction=&ConvertFromDiffSpec}.BaseColor
   ) {materialFunction=&HueShift}.Result
  )
  Multiply()
  Multiply()
  n58 = TextureCoordinate() {coordinateIndex=1}
  NamedRerouteDeclaration(None=Clamp(None=ComponentMask(None=@n58) {r=true}) {desc=Fasten Mask}) {name=BodyPart, nodeColor=<rgba 1 0 0.1994 1>}
  n59 = Floor(None=Add(A=TextureCoordinate() {coordinateIndex=1}, B=Constant() {r=0.5}))
  NamedRerouteDeclaration(None=Clamp(None=ComponentMask(None=@n59) {g=true}) {desc=Wings L R}) {name=LeftRightWingPick, nodeColor=<rgba 1 0.2174 0 1>}
  n60 = Floor(None=Add(A=TextureCoordinate() {coordinateIndex=2}, B=Constant() {r=0.5}))
  n61 = OneMinus(None=@n60)
  n62 = ComponentMask(None=@n61) {r=true}
  n63 = ComponentMask(None=@n60) {g=true}
  NamedRerouteDeclaration(None=Clamp(None=Multiply(A=@n62, B=@n63)) {desc=Wings}) {name=WingOrBodyPicker, nodeColor=<rgba 0.3367 1 0 1>}
  n64 = ComponentMask(None=@n61) {r=true}
  n65 = ComponentMask(None=@n61) {g=true}
  NamedRerouteDeclaration(None=Clamp(None=Multiply(A=@n64, B=@n65)) {desc=Body}) {name=WingOrBodyPicker2, nodeColor=<rgba 0 1 0.084 1>}
  n66 = ComponentMask(None=@n60) {r=true}
  n67 = ComponentMask(None=@n60) {g=true}
  NamedRerouteDeclaration(None=Clamp(None=Multiply(A=@n66, B=@n67)) {desc=Legs}) {name=WingOrBodyPicker3, nodeColor=<rgba 0 1 0.5747 1>}
  n68 = DynamicParameter() {parameterIndex=3, defaultValue=<rgba 1 1 0.25 1>, paramNames [Wing_FlapSpeed, Param2, Param3, MainSpeed]}
  NamedRerouteDeclaration(None=@n68.MainSpeed) {name=MainSpeed, nodeColor=<rgba 1 0.375 0 1>}
 ]
}
