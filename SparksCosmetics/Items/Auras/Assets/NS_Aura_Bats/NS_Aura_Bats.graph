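# Niagara cosmetic aura that surrounds a character with swirling bat meshes, red fire, glows, sparks and dark smoke.
# Emitters follow skeletal mesh bones or sockets by partition index, while bats rise under drag and randomly signed vortex forces.
# Arm and back flame meshes pass smoothed position and a speed-based value to Interpolate Dynamic.
# IsLargeBody adjusts sizes and offsets, time of day modulates color, and a flame-position debug emitter is disabled.
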NiagaraSystem /SparksCosmetics/Items/Auras/Assets/NS_Aura_Bats/NS_Aura_Bats
dependencies {
 usedDataInterfaces [
  &NiagaraDataInterfaceCurve
  &NiagaraDataInterfaceArrayFloat
  &NiagaraDataInterfaceColorCurve
  &NiagaraDataInterfaceArrayInt32
  &NiagaraDataInterfaceArrayFloat3
  &NiagaraDataInterfaceArrayFloat2
  &NiagaraDataInterfaceSkeletalMesh
 ]
 usedDynamicInputs [&VectorLength, /Game/Effects/Niagara/DynamicInputs/Math/FN_Fit]
 usedModules [
  &self:SystemSpawnScript.SystemScriptSource.SystemScriptGraph.NiagaraNodeAssignment_0.SetVariables_#1
  &SystemState
  &self:SystemSpawnScript.SystemScriptSource.SystemScriptGraph.NiagaraNodeAssignment_2.SetVariables_#2
  &self:SystemSpawnScript.SystemScriptSource.SystemScriptGraph.NiagaraNodeAssignment_1.SetVariables_#3
 ]
}
emitters {
 Bats {
  emitterUpdateScript [
   &EmitterState(
    Delay First Loop Only=true
    Enable Distance Culling=true
    Life Cycle Mode=::NewEnumerator1
    Loop Delay=0.05
    Loop Duration=1
    MaxDistance=15000
    Min Distance Response=::SleepAndLetParticlesFinish
    Module.bApplySpawnCountScaleByDistance=true
    Module.bUseMaxDistance=true
    Module.bUseMinDistance=true
    Reset Age On Awaken=true
    Scalability Mode=::NewEnumerator1
    Scale Spawn Count=true
    Spawn Count Scale By Distance=DataInterface() {propertyValues {Curve {keys [<at 0 1>,<at 0.95 1>,<at 1 0>]}}}
    UseLoopDelay=true
   )
   &SpawnRate(Spawn Probability=0.8, SpawnRate=&RandomRangeFloat(Minimum=2, Maximum=6), Use Spawn Probability=true) {module=SpawnRate001}
  ]
  particleSpawnScript [
   &InitializeParticle(
    Color=<rgba 0.0833 0.0833 0.0833 1>
    Lifetime=&RandomRangeFloat(Minimum=0.6, Maximum=0.8)
    Mesh Scale Mode=::NewEnumerator3
    Mesh Uniform Scale=&Multiply_Float(A=&RandomRangeFloat(Minimum=0.06, Maximum=0.12), B=1)
    UsePositionOffset=false
    Write Material Random=true
   )
   &PartitionParticles(Number Of Partitions=5)
   &SkeletalMeshLocation(
    Bone / Socket Index=$Particles.PartitionParticles.Partition
    Bone Sampling Mode=::NewEnumerator4
    MaskSpawnGroup=false
    Orientation Sampling=::NewEnumerator1
    Sampled Position Offset Coordinate Space:ENiagaraSkelMeshTransforms=::NewEnumerator2
    Skeletal Mesh=$System.Skeletal Mesh
   )
   &ShapeLocation(
    MaskSpawnGroup=false
    Non Uniform Scale=<0.6 0.2 0.2>
    Override Seed=false
    Shape Origin=$Particles.Position
    Sphere Radius=2
    UseOwnerScale=true
   )
   &AddVelocity(Velocity=&RandomRangeVector(Minimum=<0 0 150>, Maximum=<1 1 180>))
  ]
  particleUpdateScript [
   &ParticleState()
   &VortexForce(
    Origin Pull Amount:NiagaraFloat=&RandomRangeFloat(Minimum=150, Maximum=250)
    UseOriginOffset=true
    Vortex Force Amount=&SelectFloatFromArray(
     Array Sampling Mode=::NewEnumerator0
     Direct Array Index=?
     Float Selection Array=DataInterface() {propertyValues {FloatData [-800,800]}}
     Random Seed=0
    )
    Vortex Origin Offset:Vector3f=<0 0 200>
   )
   &AerodynamicDrag(
    Aerodynamic Drag=0.01
    Aerodynamic Rotational Drag=0
    Lift Contribution=0
    Rotation Strength=0
    Use Linear Drag=true
    Use Rotational Drag=true
   )
   &SolveForcesAndVelocity(Clamp Velocity=true, Limit Acceleration=false, Speed Limit=400)
   &DynamicMaterialParameters(
    Index 0 Param 1=0.05
    Index 0 Param 2=&RandomRangeFloat(Minimum=3.5, Maximum=4.85)
    Index 0 Param 3=0.18
    Index 0 Param 4=0.1
    Index 1 Param 1=2
    Index 1 Param 2=&RandomRangeFloat(Minimum=25, Maximum=5)
    Index 1 Param 3=2
    Index 1 Param 4=0.25
    Index 2 Param 1=1
    Index 2 Param 2=1
    Index 2 Param 3=2
    Index 2 Param 4=0.25
    Module.Index0Param0Write=true
    Module.Index0Param1Write=true
    Module.Index0Param2Write=true
    Module.Index0Param3Write=true
    Module.Index1Param0Write=true
    Module.Index1Param1Write=true
    Module.Index1Param2Write=true
    Module.Index1Param3Write=true
    Module.Index2Param0Write=true
    Module.Index2Param1Write=true
    Module.Index2Param2Write=true
    Module.Index2Param3Write=true
    Write Parameter Index 1=true
    Write Parameter Index 2=true
   ) {module=DynamicMaterialParameters001}
   &AlignSpriteToMeshOrientation(SetAlignment=true, SetFacing=true)
   /SetVariables(Particles.New Vector 3f:Vector3f=&Multiply_VectorByFloat(
     Vector=$Particles.Scale
     Float=&FloatFromCurve(
      FloatCurve=DataInterface() {propertyValues {Curve {keys [
          {interpMode=RCIM_Cubic, tangentMode=RCTM_User, time=0, value=0}
          {interpMode=RCIM_Cubic, tangentMode=RCTM_User, time=0.05, value=1}
          {interpMode=RCIM_Cubic, tangentMode=RCTM_User, time=0.7, value=1}
          {tangentMode=RCTM_User, time=0.9, value=0}
         ]}}}
      Scale Curve=&RandomRangeFloat(Minimum=0.8, Maximum=1)
     )
    )) {isSetParameters=true}
   &CameraOffset(Camera Offset Amount=&FloatFromCurve(
     FloatCurve=DataInterface() {propertyValues {Curve {
        defaultValue=1
        keys [{interpMode=RCIM_Cubic, tangentMode=RCTM_User, time=0.5, value=1}, {interpMode=RCIM_Cubic, tangentMode=RCTM_User, time=1, value=0.2}]
       }}}
     Scale Curve=-30
    ))
  ]
  rendererClasses [&NiagaraMeshRendererProperties]
  renderers [{rendererClass=&NiagaraMeshRendererProperties, rendererIndex=0}]
 }
 DebugFlamePosition {
  bEnabled=false
  emitterUpdateScript [
   &EmitterState()
   &SpawnBurst_Instantaneous(Spawn Count=4, Use Spawn Probability=false, UseLoopCountLimit=true)
  ]
  particleSpawnScript [&InitializeParticle(
    Color=<rgba 0.2848 0 0 1>
    Lifetime=$Emitter.CurrentLoopDuration
    Sprite Size Mode=::NewEnumerator3
    UsePositionOffset=false
    Write Material Random=true
   )]
  particleUpdateScript [
   &ParticleState(Kill Particles When Lifetime Has Elapsed=false)
   &SolveForcesAndVelocity(Clamp Velocity=false, Limit Acceleration=false)
   &SkeletalMeshLocation(
    Bone / Socket Index=7
    Bone Sampling Mode=::NewEnumerator4
    MaskSpawnGroup=false
    Sampled Position Offset=&Multiply_Vector(
     A=&Lerp_Vector(A=<7 -7 17>, B=<7 -7 -17>, Alpha=&ReturnNormalizedExecIndex())
     B=&MakeCustomVectorFromBool(Bool=$User.IsLargeBody, True Vector=<1.5 1.1 1.2>, False Vector=<1 1.1 1>)
    )
    Skeletal Mesh=$System.Skeletal Mesh
   )
  ]
  rendererClasses [&NiagaraSpriteRendererProperties]
  renderers [{rendererClass=&NiagaraSpriteRendererProperties}]
 }
 Fire_Arms {
  emitterUpdateScript [
   &EmitterState(Life Cycle Mode=::NewEnumerator1, Loop Behavior=::NewEnumerator1, Loop Duration Mode=::NewEnumerator1, UseLoopDelay=false)
   &SpawnBurst_Instantaneous(Spawn Count=6, Use Spawn Probability=false, UseLoopCountLimit=false)
   &SpawnBurst_Instantaneous(Spawn Count=6, Spawn Group=1, Use Spawn Probability=false, UseLoopCountLimit=false) {module=SpawnBurst_Instantaneous001}
  ]
  particleSpawnScript [
   &PartitionParticles(Particles Per Partition=3, Partition Mode:ENiagara_PartitionMode_v2=::NewEnumerator1)
   &InitializeParticle(
    Color=&MaskLinearColorBySpawnGroup(
     Masked Linear Color=&LerpLinearColors(
      StartColor=<rgba 0.1979 0 0.0058 1>
      EndColor=<rgba 0.4954 0 0.0143 1>
      LerpFactor=$Particles.PartitionParticles.Normalized Index In Partition
     )
     Passthrough Linear Color=<rgba 0.4954 0 0.0143 1>
    )
    Lifetime=0.1
    Mesh Scale=&MaskVectorBySpawnGroup(
     Masked Vector=&MakeVector(
      X=1
      Y=1
      Z=&Multiply_Float(
       A=&Lerp_Float(A=0.6, B=0.6, Alpha=$Particles.PartitionParticles.Normalized Index In Partition)
       B=&MakeCustomFloatFromBool(Bool=$User.IsLargeBody, True Float=1.5, False Float=1)
      )
     )
     Passthrough Vector=&MakeVector(
      X=1
      Y=1
      Z=&Multiply_Float(
       A=&Lerp_Float(A=0.4, B=0.3, Alpha=$Particles.PartitionParticles.Normalized Index In Partition)
       B=&MakeCustomFloatFromBool(Bool=$User.IsLargeBody, True Float=1.5, False Float=1)
      )
     )
    )
    Mesh Scale Mode=::NewEnumerator4
    UsePositionOffset=false
    Write Material Random=true
   )
   &InitialMeshOrientation(Mesh Orientation Mode=::NewEnumerator1, Use Rotation Vector=true)
  ]
  particleUpdateScript [
   &ParticleState(Kill Particles When Lifetime Has Elapsed=false, Let Infinitely Lived Particles Die When Emitter Deactivates=true)
   &DynamicMaterialParameters(
    Index 0 Param 1=&MaskFloatBySpawnGroup(
     Masked Float=&Multiply_Float(
      A=&Lerp_Float(A=25, B=20, Alpha=$Particles.PartitionParticles.Normalized Index In Partition)
      B=&MakeCustomFloatFromBool(Bool=$User.IsLargeBody, True Float=1.3, False Float=1)
     )
     Passthrough Float=&Multiply_Float(
      A=&Lerp_Float(A=15, B=5, Alpha=$Particles.PartitionParticles.Normalized Index In Partition)
      B=&MakeCustomFloatFromBool(Bool=$User.IsLargeBody, True Float=1.3, False Float=1)
     )
    )
    Index 0 Param 2=1
    Module.Index0Param0Write=true
    Module.Index0Param1Write=true
    Module.Index0Param2Write=true
    Module.Index0Param3Write=true
   )
   &UpdateMeshOrientation(
    Facing Direction=/Game/Effects/Niagara/DynamicInputs/Engine/FN_GetSystemAxis(Axis:ENiagaraOrientationAxis=::XAxis, Flip Axis:NiagaraBool=false)
    Facing Interpolation=false
    Orientation Method=::NewEnumerator1
    Override Quat=false
    TransformFacing=false
    TransformUp=false
   )
   &SkeletalMeshLocation(
    Bone / Socket Index=&MaskIntBySpawnGroup(
     Masked Int=&SelectIntFromArray(Direct Array Index=$Particles.PartitionParticles.Partition, Int Selection Array=DataInterface() {propertyValues {IntData [1,2]}})
     Passthrough Int=&SelectIntFromArray(Direct Array Index=$Particles.PartitionParticles.Partition, Int Selection Array=DataInterface() {propertyValues {IntData [3,4]}})
     Spawn Group Mask=0
    )
    Bone Sampling Mode=::NewEnumerator4
    MaskSpawnGroup=false
    Orientation Sampling=::NewEnumerator1
    Sampled Position Offset=&MaskVectorBySpawnGroup(
     Masked Vector=&Multiply_Vector(
      A=&Lerp_Vector(
       A=&SelectVectorFromArray(Direct Array Index=$Particles.PartitionParticles.Partition, Vector Selection Array=DataInterface() {propertyValues {FloatData []}})
       B=&SelectVectorFromArray(
        Direct Array Index=$Particles.PartitionParticles.Partition
        Vector Selection Array=DataInterface() {propertyValues {FloatData [<25 0 0>, <-25 0 0>]}}
       )
       Alpha=$Particles.PartitionParticles.Normalized Index In Partition
      )
      B=&MakeCustomVectorFromBool(Bool=$User.IsLargeBody, True Vector=<1.2 1 1>, False Vector=<1 1 1>)
     )
     Passthrough Vector=&Lerp_Vector(
      A=<0 0 0>
      B=&SelectVectorFromArray(
       Direct Array Index=$Particles.PartitionParticles.Partition
       Vector Selection Array=DataInterface() {propertyValues {FloatData [<10 0 0>, <-10 0 0>]}}
      )
      Alpha=$Particles.PartitionParticles.Normalized Index In Partition
     )
    )
    Skeletal Mesh=$System.Skeletal Mesh
   ) {module=SkeletalMeshLocation001}
   &InterpolateOverTime(
    Rate Of Change=12
    Target Vector Value:Vector3f=&ConvertPositionToVector(Input Position=$Particles.Position, Convert To Absolute Large World Vector=false)
    Variable Type=::NewEnumerator2
   )
   /SetVariables(Particles.Interpolate Dynamic=&MakeFloat4FromVectorAndFloat(
     "Vector (RGB)"=$StackContext.InterpolateOverTime.InterpolateOverTime_V3_.Moving Average
     "Float (Alpha)"=&Lerp_Float(A=1.5, B=0.25, Alpha=$System.NormalizedSpeed)
    )) {isSetParameters=true, module=SetVariables_#4}
   &ScaleColor(
    Scale RGB=&Multiply_VectorByFloat(
     Vector=<1 1 1>
     Float=&FloatFromCurve(
      FloatCurve=DataInterface() {propertyValues {Curve {keys [<at 0 1 RCIM_Cubic>,<at 0.4947 1.5 RCIM_Cubic>,<at 1 1 RCIM_Cubic>]}}}
      CurveIndex=Linked() {linkedVariable {name=NPC.FortniteNPC.TimeOfDay_0-1, type=NiagaraFloat}}
      Scale Curve=1.25
     )
    )
    ScaleA=true
    ScaleRGB=true
   )
  ]
  rendererClasses [&NiagaraMeshRendererProperties]
  renderers [{rendererClass=&NiagaraMeshRendererProperties, rendererIndex=0}]
 }
 Fire_Back {
  emitterUpdateScript [
   &EmitterState(Life Cycle Mode=::NewEnumerator1, Loop Behavior=::NewEnumerator1, Loop Duration Mode=::NewEnumerator1, UseLoopDelay=false)
   &SpawnBurst_Instantaneous(Spawn Count=6, Use Spawn Probability=false, UseLoopCountLimit=false)
  ]
  particleSpawnScript [
   &PartitionParticles(Particles Per Partition=3, Partition Mode:ENiagara_PartitionMode=::NewEnumerator1)
   &InitializeParticle(
    Color=<rgba 0.4954 0 0.0143 1>
    Lifetime=0.1
    Mesh Scale=&MakeVector(
     X=1
     Y=1
     Z=&FloatFromCurve(
      FloatCurve=DataInterface() {propertyValues {Curve {
         defaultValue=1
         keys [
          {interpMode=RCIM_Cubic, tangentMode=RCTM_User, time=0, value=0.5}
          {interpMode=RCIM_Cubic, tangentMode=RCTM_User, time=0.5, value=1}
          {tangentMode=RCTM_User, time=1, value=0.5}
         ]
        }}}
      CurveIndex=&ReturnNormalizedExecIndex()
     )
    )
    Mesh Scale Mode=::NewEnumerator4
    UsePositionOffset=false
    Write Material Random=true
   )
   &InitialMeshOrientation(Mesh Orientation Mode=::NewEnumerator1, Use Rotation Vector=true)
   &CameraOffset(Camera Offset Amount=-10) {enabled=false}
  ]
  particleUpdateScript [
   &ParticleState(Kill Particles When Lifetime Has Elapsed=false, Let Infinitely Lived Particles Die When Emitter Deactivates=true)
   &DynamicMaterialParameters(
    Index 0 Param 1=&Multiply_Float(
     A=&Lerp_Float(A=25, B=20, Alpha=$Particles.PartitionParticles.Normalized Index In Partition)
     B=&MakeCustomFloatFromBool(Bool=$User.IsLargeBody, True Float=1.3, False Float=1)
    )
    Index 0 Param 2=1
    Module.Index0Param0Write=true
    Module.Index0Param1Write=true
    Module.Index0Param2Write=true
    Module.Index0Param3Write=true
   )
   &UpdateMeshOrientation(
    Facing Direction=&FN_GetSystemAxis(Axis:ENiagaraOrientationAxis=::ZAxis, Flip Axis:NiagaraBool=false)
    Facing Interpolation=false
    Orientation Method=::NewEnumerator1
    Override Quat=false
    TransformFacing=false
    TransformUp=false
   ) {enabled=false}
   &SkeletalMeshLocation(
    Bone / Socket Index=7
    Bone Sampling Mode=::NewEnumerator4
    MaskSpawnGroup=false
    Orientation Sampling=::NewEnumerator1
    Sampled Position Offset=&Multiply_Vector(
     A=&MakeVector(
      X=7
      Y=&FloatFromCurve(
       FloatCurve=DataInterface() {propertyValues {Curve {keys [
           {arriveTangent=-10.9272, interpMode=RCIM_Cubic, leaveTangent=-10.9272, tangentMode=RCTM_User, time=0, value=-1}
           <at 0.5003 -7 RCIM_Cubic>
           {arriveTangent=8.1397, interpMode=RCIM_Cubic, leaveTangent=8.1397, tangentMode=RCTM_User, time=1, value=-1}
          ]}}}
       CurveIndex=&ReturnNormalizedExecIndex()
      )
      Z=&FloatFromCurve(
       FloatCurve=DataInterface() {propertyValues {Curve {keys [<at 0 17>,<at 1 -17>]}}}
       CurveIndex=&ReturnNormalizedExecIndex()
      )
     )
     B=&MakeCustomVectorFromBool(Bool=$User.IsLargeBody, True Vector=<1.5 1.2 1.2>, False Vector=<1 1 1>)
    )
    Skeletal Mesh=$System.Skeletal Mesh
   )
   &SkeletalMeshLocation(
    Bone / Socket Index=&SelectIntFromArray(Direct Array Index=$Particles.PartitionParticles.Partition, Int Selection Array=DataInterface() {propertyValues {IntData [1,2]}})
    Bone Sampling Mode=::NewEnumerator4
    MaskSpawnGroup=false
    Orientation Sampling=::NewEnumerator1
    Sampled Position Offset=&Lerp_Vector(
     A=&SelectVectorFromArray(Direct Array Index=$Particles.PartitionParticles.Partition, Vector Selection Array=DataInterface() {propertyValues {FloatData []}})
     B=&SelectVectorFromArray(
      Direct Array Index=$Particles.PartitionParticles.Partition
      Vector Selection Array=DataInterface() {propertyValues {FloatData [<25 0 0>, <-25 0 0>]}}
     )
     Alpha=$Particles.PartitionParticles.Normalized Index In Partition
    )
    Skeletal Mesh=$System.Skeletal Mesh
   ) {enabled=false, module=SkeletalMeshLocation001}
   &InterpolateOverTime(
    Rate Of Change=12
    Target Vector Value:Vector3f=&ConvertPositionToVector(Input Position=$Particles.Position, Convert To Absolute Large World Vector=false)
    Variable Type=::NewEnumerator2
   )
   /SetVariables(Particles.Interpolate Dynamic=&MakeFloat4FromVectorAndFloat(
     "Vector (RGB)"=$StackContext.InterpolateOverTime.InterpolateOverTime_V3_.Moving Average
     "Float (Alpha)"=&Lerp_Float(A=1.5, B=0.25, Alpha=$System.NormalizedSpeed)
    )) {isSetParameters=true, module=SetVariables_#4}
   &ScaleColor(
    Scale RGB=&Multiply_VectorByFloat(
     Vector=<1 1 1>
     Float=&FloatFromCurve(
      FloatCurve=DataInterface() {propertyValues {Curve {keys [<at 0 1 RCIM_Cubic>,<at 0.4947 1.5 RCIM_Cubic>,<at 1 1 RCIM_Cubic>]}}}
      CurveIndex=Linked() {linkedVariable {name=NPC.FortniteNPC.TimeOfDay_0-1, type=NiagaraFloat}}
      Scale Curve=1.25
     )
    )
    ScaleA=true
    ScaleRGB=true
   )
  ]
  rendererClasses [&NiagaraMeshRendererProperties]
  renderers [{rendererClass=&NiagaraMeshRendererProperties, rendererIndex=0}]
 }
 Glow {
  emitterUpdateScript [
   &EmitterState(
    Inactive Response=::NewEnumerator1
    Life Cycle Mode=::NewEnumerator1
    Loop Behavior=::NewEnumerator1
    Loop Duration Mode=::NewEnumerator1
    UseLoopDelay=false
   )
   &SpawnBurst_Instantaneous(Spawn Count=7, Use Spawn Probability=false, UseLoopCountLimit=false)
  ]
  particleSpawnScript [
   &PartitionParticles(Number Of Partitions=7)
   &InitializeParticle(
    Color=&MultiplyLinearColor(
     Linear Color=<rgba 1 0 0.0308 3>
     Scale Factor=&MakeVector4_FromFloat(Float=&FNTimeOfDayScalar(Night Scale=1, Morning Scale=1, Day Scale=1.1, Evening Scale=1))
    )
    Lifetime=3
    Sprite Size Mode=::NewEnumerator3
    Uniform Sprite Size=&Multiply_Float(
     A=&SelectFloatFromArray(
      Direct Array Index=$Particles.PartitionParticles.Partition
      Float Selection Array=DataInterface() {propertyValues {FloatData [20,40,40,40,40,40,40]}}
     )
     B=&MakeCustomFloatFromBool(Bool=$User.IsLargeBody, True Float=1.5, False Float=1)
    )
    UsePositionOffset=false
    Write Material Random=true
   )
   &CameraOffset(Camera Offset Amount=&SelectFloatFromArray(
     Direct Array Index=$Particles.PartitionParticles.Partition
     Float Selection Array=DataInterface() {propertyValues {FloatData [10,10,10,10,10]}}
    ))
   /SetVariables(Particles.GlowCameraOffset=-30) {isSetParameters=true}
   /SetVariables(Particles.GlowScale:Vector2f=&Multiply_Vector2D(
     A=&SelectVector2DFromArray(
      Direct Array Index=$Particles.PartitionParticles.Partition
      Vector 2D Selection Array=DataInterface() {propertyValues {FloatData [<90 90>, <65 65>, <65 65>, <40 40>, <40 40>, <25 25>, <25 25>]}}
     )
     B=&MakeCustomVector2DFromBool(Bool=$User.IsLargeBody, True Vector2D=<1.25 1.25>, False Vector2D=<1 1>)
    )) {isSetParameters=true}
   /SetVariables(Particles.GlowColor=&FN_Select_Color_from_Time_Of_Day(Morning=<rgba 1 0 0 0.749>, Day=<rgba 1 0 0 1>, Evening=<rgba 1 0 0 0.749>, Night=<rgba 1 0 0 0.749>)) {isSetParameters=true}
   /SetVariables(Particles.Decal Dynamic Param:Vector4f=&MakeVector4(X=5, Y=0.1, Z=0, W=0)) {isSetParameters=true}
  ]
  particleUpdateScript [
   &ParticleState(
    Kill Particles When Lifetime Has Elapsed=false
    Let Infinitely Lived Particles Die When Emitter Deactivates=true
    Loop Particles Lifetime=true
   )
   &SkeletalMeshLocation(
    Bone / Socket Index=&SelectIntFromArray(Direct Array Index=$Particles.PartitionParticles.Partition, Int Selection Array=DataInterface() {propertyValues {IntData [0,1,2,3,4,5,6]}})
    Bone Sampling Mode=::NewEnumerator4
    MaskSpawnGroup=false
    Orientation Sampling=::NewEnumerator1
    Sampled Position Offset=&SelectVectorFromArray(
     Direct Array Index=$Particles.PartitionParticles.Partition
     Vector Selection Array=DataInterface() {propertyValues {FloatData [<0 0 0>, <0 0 0>, <0 0 0>, <0 0 0>, <0 0 0>]}}
    )
    Skeletal Mesh=$System.Skeletal Mesh
   )
   &SolveForcesAndVelocity(Clamp Velocity=false, Limit Acceleration=false)
   /SetVariables(Particles.GlowScale:Vector2f=&Multiply_Vector2D(
     A=$Particles.Initial.GlowScale
     B=&Vector2DFromFloat(Value=&FloatFromCurve(FloatCurve=DataInterface() {propertyValues {Curve {keys [
           {interpMode=RCIM_Cubic, tangentMode=RCTM_User, time=0, value=0.8}
           {interpMode=RCIM_Cubic, tangentMode=RCTM_User, time=0.5, value=1}
           {tangentMode=RCTM_User, time=1, value=0.8}
          ]}}}))
    )) {isSetParameters=true}
   &ScaleColor(Scale RGB=&VectorFromFloat(Value=2), ScaleA=true, ScaleRGB=true)
  ]
  rendererClasses [&NiagaraSpriteRendererProperties]
  renderers [
   {rendererClass=&NiagaraSpriteRendererProperties}
   {rendererClass=&NiagaraSpriteRendererProperties, rendererIndex=1}
  ]
 }
 Smoke001 {
  emitterUpdateScript [
   &EmitterState()
   &SpawnRate(SpawnRate=6, Use Spawn Probability=false)
  ]
  particleSpawnScript [
   &PartitionParticles(Number Of Partitions=6)
   &InitializeParticle(
    Color=<rgba 0.0103 0 0.0003 1>
    Lifetime Max=1
    Lifetime Min=0.75
    Lifetime Mode=::NewEnumerator1
    Sprite Rotation Mode=::NewEnumerator2
    Sprite Size Mode=::NewEnumerator1
    Uniform Sprite Size Max=40
    Uniform Sprite Size Min=25
    UsePositionOffset=false
    Write Material Random=true
   ) {module=InitializeParticle001}
   /SetVariables(Particles.RandomFrame=&RandomRangeInt(Minimum=0, Maximum=63)) {isSetParameters=true}
   &DynamicMaterialParameters(
    Index 0 Param 1=&RandomRangeFloat(Minimum=0.2, Maximum=0.5)
    Module.Param0WriteEnabled=true
    Module.Param1WriteEnabled=true
    Module.Param2WriteEnabled=true
    Module.Param3WriteEnabled=true
   )
   &SkeletalMeshLocation(
    Bone / Socket Index=&SelectIntFromArray(Direct Array Index=$Particles.PartitionParticles.Partition, Int Selection Array=DataInterface() {propertyValues {IntData [1,1,1,2,2,2]}})
    Bone Sampling Mode=::NewEnumerator4
    Inherit Velocity=true
    "Inherit Velocity (Scale)"=0.5
    Inherited Velocity Speed Limit=50
    Interpolate Mesh Sampling=false
    MaskSpawnGroup=false
    Sampled Position Offset=&SelectVectorFromArray(
     Direct Array Index=$Particles.PartitionParticles.Partition
     Vector Selection Array=DataInterface() {propertyValues {FloatData [<0 0 0>, <12.5 0 0>, <25 0 0>, <0 0 0>, <-12.5 0 0>, <-25 0 0>]}}
    )
    Sampled Position Offset Coordinate Space:ENiagaraMeshTransforms=::NewEnumerator3
    Skeletal Mesh=$System.Skeletal Mesh
    VelocitySpeedLimit=true
   )
   &AddVelocity(Velocity=&RandomRangeVector(Minimum=<0 0 70>, Maximum=<0 0 125>), Velocity Speed Scale=2.5)
  ]
  particleUpdateScript [
   &ParticleState()
   &SkeletalMeshLocation(
    Bone / Socket Index=&SelectIntFromArray(Direct Array Index=$Particles.PartitionParticles.Partition, Int Selection Array=DataInterface() {propertyValues {IntData [1,1,1,2,2,2]}})
    Bone Sampling Mode=::NewEnumerator4
    MaskSpawnGroup=false
    Position Sampling=::NewEnumerator1
    Sampled Position Offset=&SelectVectorFromArray(
     Direct Array Index=$Particles.PartitionParticles.Partition
     Vector Selection Array=DataInterface() {propertyValues {FloatData [<0 0 0>, <12.5 0 0>, <25 0 0>, <0 0 0>, <-12.5 0 0>, <-25 0 0>]}}
    )
    Skeletal Mesh=$System.Skeletal Mesh
   ) {module=SkeletalMeshLocation001}
   &SkeletalMeshLocation(MaskSpawnGroup=false, OverrideRandomSeed=false, Skeletal Mesh=$System.Skeletal Mesh) {enabled=false, module=SkeletalMeshLocation002}
   &Drag(Drag=&RandomRangeFloat(Minimum=3, Maximum=10), Use Linear Drag=true, Use Rotational Drag=false)
   &GravityForce(Gravity=&RandomRangeVector(Minimum=<0 0 75>, Maximum=<0 0 150>)) {module=GravityForce002}
   &SolveForcesAndVelocity(Clamp Velocity=false, Limit Acceleration=false)
   &FN_SubUVAnimation(
    End Frame=&Add_Integer(A=$Particles.RandomFrame, B=63)
    Start Frame=$Particles.RandomFrame
    Sub UV Play Rate=0.75
    SubUV Animation Mode:ENiagara_SubUVLookupMode=::NewEnumerator3
   )
   &ScaleSpriteSize(Scale Factor=&Vector2DFromFloat(Value=&FloatFromCurve(FloatCurve=DataInterface() {propertyValues {Curve {keys [
          {arriveTangent=2.5076, interpMode=RCIM_Cubic, leaveTangent=2.5076, tangentMode=RCTM_User, time=0, value=0.3}
          {arriveTangent=0.4083, interpMode=RCIM_Cubic, leaveTangent=0.4083, tangentMode=RCTM_User, time=1, value=1}
         ]}}})))
   &ScaleColor(
    Scale Alpha=&FloatFromCurve(
     FloatCurve=DataInterface() {propertyValues {Curve {keys [
         {arriveTangent=11.4042, interpMode=RCIM_Cubic, leaveTangent=11.4042, tangentMode=RCTM_User, time=0, value=0}
         {arriveTangent=0.3312, interpMode=RCIM_Cubic, leaveTangent=-2.6189, tangentMode=RCTM_Break, time=0.1744, value=0.9882}
         {arriveTangent=-0.5419, interpMode=RCIM_Cubic, leaveTangent=-0.5419, tangentMode=RCTM_User, time=1, value=0}
        ]}}}
     Scale Curve=&RandomRangeFloat(Minimum=0.6, Maximum=1)
    )
    ScaleA=true
    ScaleRGB=true
   ) {module=ScaleColor001}
   &DynamicMaterialParameters(
    Index 0 Param 1=&FloatFromCurve(
     FloatCurve=DataInterface() {propertyValues {Curve {keys [
         {arriveTangent=1.6795, interpMode=RCIM_Cubic, leaveTangent=1.6795, tangentMode=RCTM_User, time=0, value=0}
         {arriveTangent=-0.1842, interpMode=RCIM_Cubic, leaveTangent=-0.1842, tangentMode=RCTM_User, time=1, value=1}
        ]}}}
     Scale Curve=0.75
    )
    Module.Index0Param0Write=true
    Module.Index0Param1Write=true
    Module.Index0Param2Write=true
    Module.Index0Param3Write=true
   ) {module=DynamicMaterialParameters001}
   &InterpolateOverTime(Rate Of Change=7, Target Position Value:NiagaraPosition=$Particles.Position, Variable Type=::NewEnumerator6)
   /SetVariables(Particles.Position=$Particles.InterpolateOverTime.InterpolateOverTime_Position.Moving Average) {isSetParameters=true}
  ]
  rendererClasses [&NiagaraSpriteRendererProperties]
  renderers [{rendererClass=&NiagaraSpriteRendererProperties}]
 }
 Sparks {
  emitterUpdateScript [
   &EmitterState(
    Inactive Response=::NewEnumerator1
    Life Cycle Mode=::NewEnumerator1
    Loop Behavior=::NewEnumerator1
    Loop Duration Mode=::NewEnumerator1
    UseLoopDelay=false
   )
   &SpawnRate(SpawnRate=42, Use Spawn Probability=false)
  ]
  particleSpawnScript [
   &PartitionParticles(Number Of Partitions=5)
   &InitializeParticle(
    Lifetime Max=0.75
    Lifetime Min=0.5
    Lifetime Mode=::NewEnumerator1
    Position Mode=::NewEnumerator1
    Sprite Rotation Mode=::NewEnumerator2
    Sprite Size=&Lerp_Vector2(
     A=<1 2>
     B=<2 4.5>
     Alpha=&FloatFromCurve(
      FloatCurve=DataInterface() {propertyValues {Curve {keys [
          {interpMode=RCIM_Cubic, tangentMode=RCTM_User, time=0, value=0}
          {interpMode=RCIM_Cubic, tangentMode=RCTM_User, time=0.3, value=0}
          {interpMode=RCIM_Cubic, tangentMode=RCTM_User, time=0.8, value=1}
         ]}}}
      CurveIndex=&RandomRangeFloat(Minimum=0, Maximum=1)
     )
    )
    Sprite Size Mode=::NewEnumerator4
    UsePositionOffset=false
    Write Material Random=true
   ) {module=InitializeParticle001}
   &ShapeLocation(
    Apply To Particle Position=::NewEnumerator1
    MaskSpawnGroup=false
    Override Seed=false
    Shape Origin=&VectorFromFloat(Value=0)
    Sphere Radius=5
    UseOwnerScale=true
   )
   &AddVelocity(
    Velocity=&MakeVector(
     X=&RandomRangeFloat(Minimum=-20, Maximum=20)
     Y=&RandomRangeFloat(Minimum=-20, Maximum=20)
     Z=&RandomRangeFloat(Minimum=75, Maximum=150)
    )
    Velocity Speed Scale=&MakeCustomFloatFromBool(Bool=$User.IsLargeBody, True Float=1.2, False Float=0.8)
   ) {module=AddVelocity001}
   &CameraOffset(Camera Offset Amount=-20)
  ]
  particleUpdateScript [
   &ParticleState()
   &ScaleColor(
    Scale Alpha=&Lerp_Float(
     A=1
     B=0
     Alpha=&FloatFromCurve(FloatCurve=DataInterface() {propertyValues {Curve {keys [
          {arriveTangent=3.1144, interpMode=RCIM_Cubic, leaveTangent=3.1144, tangentMode=RCTM_User, time=0, value=0}
          {interpMode=RCIM_Cubic, tangentMode=RCTM_User, time=1, value=1}
         ]}}})
    )
    Scale RGB=&Multiply_VectorByFloat(
     Vector=<1 1 1>
     Float=&FloatFromCurve(
      FloatCurve=DataInterface() {propertyValues {Curve {keys [
          {arriveTangent=-12.6044, interpMode=RCIM_Cubic, leaveTangent=-12.6044, tangentMode=RCTM_User, time=0, value=10}
          {interpMode=RCIM_Cubic, tangentMode=RCTM_User, time=1, value=1}
         ]}}}
      Scale Curve=&FloatFromCurve(
       FloatCurve=DataInterface() {propertyValues {Curve {keys [<at 0 1 RCIM_Cubic>,<at 0.5002 1.5 RCIM_Cubic>,<at 1 1 RCIM_Cubic>]}}}
       CurveIndex=Linked() {linkedVariable {name=NPC.FortniteNPC.TimeOfDay_0-1, type=NiagaraFloat}}
       Scale Curve=2
      )
     )
    )
    ScaleA=true
    ScaleRGB=true
   )
   &Color(Color=&MultiplyLinearColor(
     Linear Color=&ColorFromCurve(ColorCurve=DataInterface() {propertyValues {
        AlphaCurve {keys [<at 0 0.6 RCIM_Cubic>,<at 0.244 0.6>,<at 1 0 RCIM_Cubic>]}
        BlueCurve {keys [<at 0 0.044 RCIM_Cubic>,<at 0.2395 0.0147>,<at 0.8355 0 RCIM_Cubic>]}
        GreenCurve {keys [<at 0 0 RCIM_Cubic>,<at 0.2395 0>,<at 0.8355 0 RCIM_Cubic>]}
        RedCurve {keys [<at 0 1.5 RCIM_Cubic>,<at 0.2395 0.5>,<at 0.8355 0 RCIM_Cubic>]}
       }})
     Scale Factor=&MakeVector4_FromFloat(Float=&FNTimeOfDayScalar(
       Night Scale=1.5
       Morning Scale=1.75
       Day Scale=&FN_FrontEndFloatModifier(Gameplay=2, Front End=1)
       Evening Scale=1.75
      ))
    ))
   &ScaleSpriteSize(Uniform Curve Sprite Scale=DataInterface() {propertyValues {Curve {keys [
        {arriveTangent=0.0565, interpMode=RCIM_Cubic, leaveTangent=0.0565, tangentMode=RCTM_User, time=0, value=1}
        {arriveTangent=-0.025, interpMode=RCIM_Cubic, leaveTangent=-0.025, tangentMode=RCTM_User, time=1, value=0}
       ]}}}) {module=ScaleSpriteSize001}
   &SkeletalMeshLocation(
    Bone / Socket Index=&SelectIntFromArray(Direct Array Index=$Particles.PartitionParticles.Partition, Int Selection Array=DataInterface() {propertyValues {IntData [1,2,3,4]}})
    Bone Sampling Mode=::NewEnumerator4
    MaskSpawnGroup=false
    Position Sampling=::NewEnumerator1
    Sampled Position Offset=$StackContext.ShapeLocation.ShapePosition
    Skeletal Mesh=$System.Skeletal Mesh
    Source Movement Scale:Vector3f=<0 0 0>
   ) {module=SkeletalMeshLocation001}
   &Drag(Drag=1.5, Use Linear Drag=true, Use Rotational Drag=false)
   &CurlNoiseForce(
    Curl Noise Cone Mask Angle=90
    Curl Noise Cone Mask Falloff Angle=35
    Mask Curl Noise=true
    Noise Strength=500
    Pan Noise=true
    Pan Noise Field=<0.25 0.25 0.25>
   )
   &SolveForcesAndVelocity(Clamp Velocity=false, Limit Acceleration=false)
  ]
  rendererClasses [&NiagaraSpriteRendererProperties]
  renderers [{rendererClass=&NiagaraSpriteRendererProperties}]
 }
}
summary {
 systemName=NS_Aura_Bats
 emitters [
  {emitterName=Bats, rendererClasses [&NiagaraMeshRendererProperties]}
  {emitterName=Sparks, rendererClasses [&NiagaraSpriteRendererProperties]}
  {emitterName=Smoke001, rendererClasses [&NiagaraSpriteRendererProperties]}
  {emitterName=Glow, rendererClasses [&NiagaraSpriteRendererProperties]}
  {bEnabled=false, emitterName=DebugFlamePosition, rendererClasses [&NiagaraSpriteRendererProperties]}
  {emitterName=Fire_Back, rendererClasses [&NiagaraMeshRendererProperties]}
  {emitterName=Fire_Arms, rendererClasses [&NiagaraMeshRendererProperties]}
 ]
 userVariables [
  User.bIsFrontEnd {type=NiagaraBool}
  User.bIsFrontEndPreview {type=NiagaraBool}
  User.Interpolate Intensity {type=NiagaraFloat, defaultValue=1, description=""}
  User.IsFrontEnd {type=NiagaraBool}
  User.IsLargeBody {type=NiagaraBool}
 ]
}
